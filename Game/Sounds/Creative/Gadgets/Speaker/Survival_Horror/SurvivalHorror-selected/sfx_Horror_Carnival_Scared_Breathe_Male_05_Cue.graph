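# A sound cue for a male scared-breathing horror carnival effect, filed among the Creative speaker gadget's Survival Horror sounds.
# A single wave player feeds the sfx_Horror_Carnival_Scared_Breathe_Male_05 wave straight to the output, with no randomization or modulation nodes.
# It plays in the SFX sound class and uses the shared tile-ambient attenuation asset for spatial falloff.
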SoundCue /Game/Sounds/Creative/Gadgets/Speaker/Survival_Horror/SurvivalHorror-selected/sfx_Horror_Carnival_Scared_Breathe_Male_05_Cue
root=WavePlayer_0
nodes [{
  id=WavePlayer_0
  properties {soundWaveAssetPtr=&/sfx_Horror_Carnival_Scared_Breathe_Male_05}
 }]
properties {
 attenuationSettings=/Game/Sounds/Attenuation/Fort_TileAmbient_Att
 maxDistance=10000
 soundClassObject=&SFX
}
