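# Sound cue that plays a looping scared heartbeat from the horror carnival set, stored among the Creative speaker gadget's Survival Horror sounds.
# A single looping wave player is the root, with no randomization, mixing or modulation nodes.
# It routes to the SFX sound class with a boosted volume and uses the shared tile ambient attenuation for spatial falloff.
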SoundCue /Game/Sounds/Creative/Gadgets/Speaker/Survival_Horror/SurvivalHorror-selected/sfx_Horror_Carnival_Scared_Heartbeat_01_LOOP_Cue
root=WavePlayer_0
nodes [{
  id=WavePlayer_0
  properties {
   bLooping=true
   soundWaveAssetPtr=&/sfx_Horror_Carnival_Scared_Heartbeat_01
  }
 }]
properties {
 attenuationSettings=/Game/Sounds/Attenuation/Fort_TileAmbient_Att
 maxDistance=10000
 soundClassObject=&SFX
 volumeMultiplier=1.5
}
